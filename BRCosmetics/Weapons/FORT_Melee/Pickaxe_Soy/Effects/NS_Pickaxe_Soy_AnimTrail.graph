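# Niagara system that draws the swing trail ribbon for the Soy pickaxe melee cosmetic.
# A single FN_AnimTrail_V4 emitter spawns, initializes and updates short-lived, very low-alpha white particles using Fortnite AnimTrail modules with bone or socket indices and a skeletal mesh user parameter.
# Curves fade alpha and taper ribbon width over life, with disabled alternate versions kept in the stack.
# Dynamic material parameters feed engine time and constants to the ribbon material.
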NiagaraSystem /BRCosmetics/Weapons/FORT_Melee/Pickaxe_Soy/Effects/NS_Pickaxe_Soy_AnimTrail
dependencies {usedDataInterfaces [&NiagaraDataInterfaceCurve], usedDynamicInputs [], usedModules [&SystemState]}
emitters {FN_AnimTrail_V4 {
  emitterUpdateScript [
   &EmitterState()
   &FN_SpawnAnimTrail_v2()
  ]
  particleSpawnScript [
   /Game/Effects/Niagara/Modules/AnimTrail/FN_InitializeAnimTrail/FN_InitializeAnimTrailBoneIndecies(First Bone or Socket Index=1)
   /Game/Effects/Niagara/Modules/AnimTrail/FN_InitializeAnimTrail/FN_InitializeAnimTrail(AnimTrailWidth=1, Color=<rgba 1 1 1 0.003>, Lifetime=0.35)
  ]
  particleUpdateScript [
   &SolveForcesAndVelocity(Clamp Velocity=false, Limit Acceleration=false)
   /Game/Effects/Niagara/Modules/AnimTrail/FN_UpdateAnimTrail/FN_UpdateAnimTrail()
   /Game/Effects/Niagara/Modules/Materials/Material_v2() {enabled=false}
   &ScaleColor(
    Scale Alpha=&FloatFromCurve(FloatCurve=DataInterface() {propertyValues {Curve {keys [
         {arriveTangent=-17.1094, interpMode=RCIM_Cubic, leaveTangent=-17.1094, tangentMode=RCTM_User, time=0.005, value=3}
         {arriveTangent=-0.4835, interpMode=RCIM_Cubic, leaveTangent=-0.4835, tangentMode=RCTM_User, time=0.2877, value=0.1515}
         {arriveTangent=-0.0191, interpMode=RCIM_Cubic, leaveTangent=-0.0191, tangentMode=RCTM_User, time=1.0011, value=0}
        ]}}})
    ScaleA=true
    ScaleRGB=false
   )
   &ScaleColor(
    Scale Alpha=&FloatFromCurve(FloatCurve=DataInterface() {propertyValues {Curve {keys [
         {arriveTangent=-17.1094, time=0.005, value=3}
         {arriveTangent=-0.4835, leaveTangent=-0.4835, time=0.4693, value=0.0645}
         <at 0.811 2.897>
         {arriveTangent=-0.0191, interpMode=RCIM_Cubic, leaveTangent=-0.0191, tangentMode=RCTM_User, time=1.0011, value=0}
        ]}}})
    ScaleA=true
    ScaleRGB=false
   ) {enabled=false, module=ScaleColor001}
   &ScaleRibbonWidth(Ribbon Width Scale=&FloatFromCurve(FloatCurve=DataInterface() {propertyValues {Curve {keys [
         {arriveTangent=-0.3, interpMode=RCIM_Cubic, time=0.1, value=1}
         {arriveTangent=-0.0294, interpMode=RCIM_Cubic, leaveTangent=-0.0294, tangentMode=RCTM_User, time=0.25, value=0.9932}
         {arriveTangent=-1.5611, interpMode=RCIM_Cubic, leaveTangent=-1.5611, tangentMode=RCTM_User, time=1, value=0}
        ]}}}))
   &ScaleRibbonWidth(Ribbon Width Scale=&FloatFromCurve(FloatCurve=DataInterface() {propertyValues {Curve {keys [
         {arriveTangent=-0.3, interpMode=RCIM_Cubic, time=0.1, value=1}
         <at 0.2917 0.104>
         {arriveTangent=-0.0294, interpMode=RCIM_Cubic, leaveTangent=-0.0294, tangentMode=RCTM_User, time=0.5, value=0.9932}
         {arriveTangent=-1.5611, interpMode=RCIM_Cubic, leaveTangent=-1.5611, tangentMode=RCTM_User, time=1, value=0}
        ]}}})) {enabled=false, module=ScaleRibbonWidth001}
   &DynamicMaterialParameters(
    Index 0 Param 2=&Multiply_Float(A=1, B=$Engine.Time)
    Index 0 Param 3=0.1
    Index 1 Param 1=0.05
    Index 1 Param 3=0.5
    Index 1 Param 4=0.5
    Module.Param0WriteEnabled=true
    Module.Param0WriteEnabled1=true
    Module.Param1WriteEnabled=true
    Module.Param1WriteEnabled1=true
    Module.Param2WriteEnabled=true
    Module.Param2WriteEnabled1=true
    Module.Param3WriteEnabled=true
    Module.Param3WriteEnabled1=true
    Write Parameter Index 1=true
   )
  ]
  rendererClasses [&NiagaraRibbonRendererProperties]
  renderers [{rendererClass=&NiagaraRibbonRendererProperties}]
 }}
summary {
 systemName=NS_Pickaxe_Soy_AnimTrail
 emitters [{emitterName=FN_AnimTrail_V4, rendererClasses [&NiagaraRibbonRendererProperties]}]
 userVariables [
  User.bIsFrontEndPreview {type=NiagaraBool}
  User.PlayerSpeed {type=NiagaraFloat, defaultValue=0, description=""}
  User.RibbonIDOffset {type=NiagaraInt32}
  User.SkeletalMesh {
   type=NiagaraDataInterfaceSkeletalMesh
   defaultValue {
    struct=NiagaraExt_VariableValue_DataInterface
    value {dataInterface=&self:NiagaraDataInterfaceSkeletalMesh_2}
   }
  }
 ]
}
