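# Niagara trail effect for the Soy pickaxe that emits small, twinkling light-blue star sprites.
# One continuous-rate emitter spawns particles in a small sphere with randomized color, lifetime, size and rotation, slight inherited and random velocity, and light gravity.
# Jagged curves on alpha and sprite size create the flicker, with a second curve shrinking size over life.
# Curl noise and drag are disabled, and none of the shown modules read the user parameters.
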NiagaraSystem /BRCosmetics/Weapons/FORT_Melee/Pickaxe_Soy/Effects/NS_Pickaxe_Soy_FlareTrail
dependencies {
 usedDataInterfaces [&NiagaraDataInterfaceCurve]
 usedDynamicInputs []
 usedModules [/Niagara/Modules/System/SystemLifeCycle, &SystemState]
}
emitters {FN_MultiStars {
  emitterUpdateScript [
   &EmitterState()
   &SpawnRate(SpawnRate=200, Use Spawn Probability=false)
  ]
  particleSpawnScript [
   /Niagara/Modules/Spawn/Initialization/InitializeParticle(
    Color=&UniformRangedLinearColor(Minimum=<rgba 0.085 0.695 1 0.01>, Maximum=<rgba 0.23 0.8717 1 1>, Color Channel Mode=::NewEnumerator1)
    Lifetime=&UniformRangedFloat(
     Minimum=0.1
     Maximum=&UniformRangedFloat(Minimum=0.2, Maximum=&UniformRangedFloat(Minimum=0.2, Maximum=0.75))
    )
    Module.Write Color=true
    Module.Write Lifetime=true
    Module.Write Mass=true
    Module.Write Material Random=true
    Module.Write Position=false
    Module.Write Scale=false
    Module.Write SpriteRotation=true
    Module.Write SpriteSize=true
    Position=&SimulationPosition()
    Sprite Rotation=&UniformRangedFloat(Minimum=-20, Maximum=20)
    Sprite Size=&Vector2DFromFloat(Value=&UniformRangedFloat(Minimum=10, Maximum=50))
   )
   &InheritVelocity(Module.Limit Speed=false, Velocity Scale=<0.01 0.01 0.01>)
   &AddVelocity(Velocity=&UniformRangedVector(Minimum=<-3 -3 -3>, Maximum=<3 3 3>))
   /Niagara/Modules/Spawn/Location/SphereLocation(MaskSpawnGroup=false, Module.Override Randomness=false, Sphere Radius=5, Surface Only=false)
  ]
  particleUpdateScript [
   &ParticleState()
   &GravityForce(Gravity=<0 0 -50>)
   /Niagara/Modules/Update/Forces/CurlNoiseForce(
    Noise Frequency:NiagaraFloat=&UniformRangedFloat(Minimum=0.1, Maximum=1)
    Noise Strength=&UniformRangedFloat(Minimum=300, Maximum=600)
    Pan Noise=false
    Pan Noise Field=<1 1 0>
   ) {enabled=false}
   &Drag(Drag=10, Use Linear Drag=true, Use Rotational Drag=false) {enabled=false}
   &SolveForcesAndVelocity(Clamp Velocity=false, Limit Acceleration=false)
   &ScaleColor(
    Scale Alpha=&FloatFromCurve(FloatCurve=DataInterface() {propertyValues {Curve {keys [<at 0 1>,<at 0.1 0.1331>,<at 0.3 1.0073>,<at 0.4 0.1331>,<at 0.4447 0.5627>,<at 0.5388 0.267>,<at 0.5705 0.3986>,<at 0.6388 0.0123>,<at 0.8 0.9523>,<at 0.879 0.6403>,<at 1.0011 0.0032>]}}})
    ScaleA=true
    ScaleRGB=false
   )
   &ScaleSpriteSize(Scale Factor=&Vector2DFromFloat(Value=&Multiply_Float(
      A=&FloatFromCurve(FloatCurve=DataInterface() {propertyValues {Curve {keys [<at 0 1>,<at 0.1823 0.1407>,<at 0.3163 0.5798>,<at 0.4103 0.2841>,<at 0.4967 0.4615>,<at 0.6388 0.0123>,<at 0.8 0.9523>,<at 0.9 0.0712>,<at 1 1>]}}})
      B=&FloatFromCurve(FloatCurve=DataInterface() {propertyValues {Curve {keys [<at 0 1>,<at 1 0.25>]}}})
     ))) {module=ScaleSpriteSize001}
  ]
  rendererClasses [&NiagaraSpriteRendererProperties]
  renderers [{rendererClass=&NiagaraSpriteRendererProperties}]
 }}
summary {
 systemName=NS_Pickaxe_Soy_FlareTrail
 emitters [{emitterName=FN_MultiStars, rendererClasses [&NiagaraSpriteRendererProperties]}]
 userVariables [
  User.bIsFrontEndPreview {type=NiagaraBool}
  User.GlobalOffset {type=Vector3f, description="", defaultValue {struct=Vector3f, value=<0 0 0>}}
  User.PlayerSpeed {type=NiagaraFloat, defaultValue=0, description=""}
  User.RibbonIDOffset {type=NiagaraInt32}
  User.SkeletalMesh {
   type=NiagaraDataInterfaceSkeletalMesh
   defaultValue {
    struct=NiagaraExt_VariableValue_DataInterface
    value {dataInterface=&self:NiagaraDataInterfaceSkeletalMesh_0}
   }
  }
 ]
}
